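# Sound cue for the Bone Shotgun's aim-down-sights "in" foley.
# A random node picks one of two equally weighted wave variations and feeds a modulator that varies pitch.
# It plays at reduced volume in the weapon foley sound class and borrows the weapon reload attenuation settings.
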SoundCue /PrimalGameplay/Items/Guns/Bone/BoneShotgun/Shotgun_Bone_ADS_In
root=Modulator_0
nodes [
 {id=Modulator_0, children [Random_0], properties {pitchMax=1.1}}
 {id=Random_0, children [WavePlayer_2,WavePlayer_4], properties {weights [1,1]}}
 {id=WavePlayer_2, properties {soundWaveAssetPtr=&/Shotgun_Bone_ADS_In_01}}
 {id=WavePlayer_4, properties {soundWaveAssetPtr=&/Shotgun_Bone_ADS_In_02}}
]
properties {
 attenuationSettings=&Fort_Weapon_Reload_Att
 maxDistance=2250
 soundClassObject=&Fort_WeaponFoley
 volumeMultiplier=0.501
}
