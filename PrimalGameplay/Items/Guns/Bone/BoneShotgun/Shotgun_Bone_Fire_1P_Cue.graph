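# First-person fire sound cue for the Bone Shotgun, built from close-range gunshot waves plus a growl layer.
# A mixer plays two layers at equal volume, each picking randomly from four wave variants through a modulator that varies pitch only.
# It uses first-person ranged-weapon fire sound class, projectile attenuation and gunfire concurrency assets, with overall volume reduced.
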SoundCue /PrimalGameplay/Items/Guns/Bone/BoneShotgun/Shotgun_Bone_Fire_1P_Cue
root=Mixer_2
nodes [
 {id=Mixer_2, children [Modulator_1,Modulator_5], properties {inputVolume [1,1]}}
 {id=Modulator_1, children [Random_1], properties {pitchMax=1.2, pitchMin=0.9, volumeMax=1, volumeMin=1}}
 {id=Modulator_5, children [Random_6], properties {pitchMax=1.2, pitchMin=0.9, volumeMax=1, volumeMin=1}}
 {id=Random_1, children [WavePlayer_5, WavePlayer_6, WavePlayer_7, WavePlayer_11], properties {weights [1,1,1,1]}}
 {id=Random_6, children [WavePlayer_22, WavePlayer_23, WavePlayer_24, WavePlayer_25], properties {weights [1,1,1,1]}}
 {id=WavePlayer_11, properties {soundWaveAssetPtr=&/Shotgun_Bone_Fire_Close_04}}
 {id=WavePlayer_22, properties {soundWaveAssetPtr=&/Shotgun_Bone_Fire_Close_Growl_01}}
 {id=WavePlayer_23, properties {soundWaveAssetPtr=&/Shotgun_Bone_Fire_Close_Growl_02}}
 {id=WavePlayer_24, properties {soundWaveAssetPtr=&/Shotgun_Bone_Fire_Close_Growl_03}}
 {id=WavePlayer_25, properties {soundWaveAssetPtr=&/Shotgun_Bone_Fire_Close_Growl_04}}
 {id=WavePlayer_5, properties {soundWaveAssetPtr=&/Shotgun_Bone_Fire_Close_01}}
 {id=WavePlayer_6, properties {soundWaveAssetPtr=&/Shotgun_Bone_Fire_Close_02}}
 {id=WavePlayer_7, properties {soundWaveAssetPtr=&/Shotgun_Bone_Fire_Close_03}}
]
properties {
 attenuationSettings=&Fort_Weapon_Projectile_FirstPerson_Athena_Att
 maxDistance=2097152
 soundClassObject=&Fort_WeaponRanged_Fire_1P_PSM
 volumeMultiplier=0.65
 attenuationOverrides {distanceAlgorithm=NaturalSound, falloffDistance=5000, attenuationShapeExtents {x=1000}}
 concurrencyOverrides {bLimitToOwner=true, maxCount=3, resolutionRule=StopOldest}
 concurrencySet [&Weapon_Gunfire_FirstPerson_Conc]
}
